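# Unlit, translucent, two-sided particle material for toon speedline effects in the Pickaxe_ParrotPen cosmetic, usable on Niagara sprites and mesh particles.
# Particle color RGB passes through the shared emissive adjustments function to provide the glow color.
# A dynamic parameter, TexOffset, shifts the toon lines texture along U, letting the particle system slide or animate the lines.
# The texture's red channel is multiplied by particle alpha to form the opacity mask.
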Material /BRCosmetics/Weapons/FORT_Melee/Pickaxe_ParrotPen/FX/M_FX_Speedlines
graph {
 properties {
  bUsedWithNiagaraMeshParticles=true
  bUsedWithNiagaraSprites=true
  blendMode=BLEND_Translucent
  pixelDepthOffsetMode=PDOM_Legacy
  shadingModel=MSM_Unlit
  twoSided=true
  usedShadingModels=Unlit
 }
 nodes [
  n1 = ParticleColor()
  MaterialFunctionCall(Emissive=@n1.RGB) {materialFunction=&MF_FX_EmissiveAdjustments}
  n2 = TextureCoordinate()
  n3 = MaterialFunctionCall(X=DynamicParameter() {defaultValue=<rgba 0 1 1 1>, paramNames [TexOffset,Param2,Param3,Param4]}.TexOffset) {materialFunction=&MakeFloat2}
  Multiply(
   A=TextureSample(UVs=Add(A=@n2, B=@n3.Result) {constB=0}) {samplerType=SAMPLERTYPE_LinearColor, texture=&/T_FX_Toon_Lines}.R
   B=@n1.A
  )
 ]
}
